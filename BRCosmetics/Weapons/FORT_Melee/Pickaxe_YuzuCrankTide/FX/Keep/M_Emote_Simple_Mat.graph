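# Unlit translucent Niagara sprite material for YuzuCrankTide pickaxe FX, drawing an animated radial sweep with a grid pattern.
# FX_Progress pans a radially mapped gradient, warped by a normal texture with per-particle random offset and FX_Distort-driven scale.
# Gradient times grid indexes a curve-atlas color ramp scaled by FX_Temperature, plus a sharp, tinted hot edge.
# Opacity combines the gradient, particle alpha, a sphere mask and depth fade.
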Material /BRCosmetics/Weapons/FORT_Melee/Pickaxe_YuzuCrankTide/FX/Keep/M_Emote_Simple_Mat
graph {
 properties {
  bAutomaticallySetUsageInEditor=false
  bUsedWithNiagaraSprites=true
  blendMode=BLEND_Translucent
  shadingModel=MSM_Unlit
  usedShadingModels=Unlit
 }
 nodes [
  n1 = ScalarParameter() {parameterName=LUT_MAX_LOW}
  n2 = ScalarParameter() {parameterName=LUT MAX_HIGH}
  n3 = DynamicParameter() {defaultValue=<rgba 0 1 1 1>, paramNames [FX_Progress, FX_Temperature, FX_Distort, FX_Grid_Strength]}
  n4 = TextureSampleParameter2D(UVs=Multiply(A=TextureCoordinate(), B=ScalarParameter() {defaultValue=13, parameterName=Grid_Count}) {constB=13}) {parameterName=Main_TX, texture=/BRCosmetics/Weapons/FORT_Melee/Pickaxe_YuzuCrankTide/FX/T_YuzuCrankTide_Grid}
  n5 = Multiply(
   A=Subtract(A=TextureCoordinate()) {constB=0.5}
   B=MaterialFunctionCall(X=Constant() {r=1}, Y=Constant() {r=1}) {materialFunction=&MakeFloat2}.Result
  )
  n6 = ScalarParameter() {defaultValue=1, parameterName=Main_Tile_X}
  n7 = Constant() {r=1}
  n8 = OneMinus(None=Multiply(
    A=MaterialFunctionCall(Vector or UVs=@n5) {materialFunction=&VectorToRadialValue}.Radial Coordinates
    B=MaterialFunctionCall(X=@n6, Y=@n7) {materialFunction=&MakeFloat2}.Result
   ))
  n9 = ScalarParameter() {defaultValue=-0.7, parameterName=Pan_Min}
  n10 = ScalarParameter() {defaultValue=0.3, parameterName=Pan_Max}
  n11 = DynamicParameter() {defaultValue=<rgba 0 1 1 1>, paramNames [FX_Progress, FX_Temperature, FX_Distort, FX_Grid_Strength]}
  n12 = MaterialFunctionCall(
   X=Constant()
   Y=Add(A=LinearInterpolate(A=@n9, B=@n10, Alpha=@n11.FX_Progress) {constA=-0.7, constB=0.3}, B=ScalarParameter() {parameterName=Pan_Offset})
  ) {materialFunction=&MakeFloat2}
  n13 = Subtract(A=TextureCoordinate()) {constB=0.5}
  n14 = ScalarParameter() {parameterName=UV_Distort_Scale_LOW}
  n15 = DynamicParameter() {defaultValue=<rgba 0 1 1 1>, paramNames [FX_Progress, FX_Temperature, FX_Distort, FX_Grid_Strength]}
  n16 = MaterialFunctionCall(
   X=LinearInterpolate(A=ScalarParameter() {defaultValue=1, parameterName=UV_Distort_X}, B=@n14, Alpha=@n15.FX_Distort)
   Y=LinearInterpolate(A=ScalarParameter() {defaultValue=1, parameterName=UV_Distort_Y}, B=@n14, Alpha=@n15.FX_Distort)
  ) {materialFunction=&MakeFloat2}
  n17 = ParticleRandom()
  n18 = ParticleRandom()
  n19 = TextureCoordinate()
  n20 = TextureSampleParameter2D!(
   UVs=Add(
    A=Multiply(A=@n13, B=@n16.Result)
    B=MaterialFunctionCall(X=@n17, Y=@n18) {materialFunction=&MakeFloat2}.Result
   )
   "DDX(UVs)"=DDX(Value=@n19)
   "DDY(UVs)"=DDY(Value=@n19)
   Apply View MipBias
  ) {
   mipValueMode=TMVM_Derivative
   parameterName=TX_Distort
   samplerType=SAMPLERTYPE_Normal
   texture=/BRCosmetics/Weapons/FORT_Melee/Pickaxe_YuzuCrankTide/FX/T_YuzuCrankTide_Normal
  }
  n21 = MaterialFunctionCall(X=@n20.R, Y=@n20.R) {materialFunction=&MakeFloat2}
  n22 = ScalarParameter() {defaultValue=0.05, parameterName=Distort_Strength}
  n23 = TextureCoordinate()
  n24 = TextureSampleParameter2D(UVs=Add(A=Add(A=@n8, B=@n12.Result), B=Multiply(A=@n21.Result, B=@n22)), "DDX(UVs)"=DDX(Value=@n23), "DDY(UVs)"=DDY(Value=@n23)) {
   mipValueMode=TMVM_Derivative
   parameterName=TX_Line
   samplerSource=SSM_Clamp_WorldGroupSettings
   texture=/Game/Effects/Fort_Effects/Textures/Grad/T_FX_LinearSweep_Gradient
  }
  n25 = Multiply(
   A=Multiply(A=Saturate(None=@n4.RGB), B=ScalarParameter() {defaultValue=1, parameterName=Grid_Strength})
   B=Multiply(A=@n24.RGB, B=ScalarParameter() {defaultValue=1, parameterName=Grid_Strength})
  )
  n26 = CurveAtlasRowParameter(CurveTime=LinearInterpolate(
    A=ScalarParameter() {defaultValue=0.2, parameterName=LUT MIN}
    B=LinearInterpolate(A=@n1, B=@n2, Alpha=@n3.FX_Temperature)
    Alpha=Saturate(None=@n25)
   )) {
   atlas=/BRCosmetics/Weapons/FORT_Melee/Pickaxe_YuzuCrankTide/FX/CA_Pickaxe_YuzuCrankTide
   curve=/BRCosmetics/Weapons/FORT_Melee/Pickaxe_YuzuCrankTide/FX/C_Pickaxe_YuzuCrankTide
   parameterName=BaseLUT
  }
  n27 = LinearInterpolate(
   A=VectorParameter() {parameterName=Hot_Edge_Color_Mult_Min, defaultValue=<rgba 1 1 1 1>}.RGB
   B=VectorParameter() {parameterName=Hot_Edge_Color_Mult_Max, defaultValue=<rgba 1 1 1 1>}.RGB
   Alpha=DynamicParameter() {defaultValue=<rgba 0 1 1 1>, paramNames [FX_Progress, FX_Temperature, FX_Distort, FX_Grid_Strength]}.FX_Temperature
  )
  n28 = Power(Base=@n24.RGB, Exp=ScalarParameter() {defaultValue=300, parameterName=Hot_Edge_Exp}) {constExponent=300}
  Add(A=@n26, B=Multiply(A=Multiply(A=@n27, B=@n28), B=ScalarParameter() {defaultValue=0.9, parameterName=Hot_Edge_Strength}))
  LinearInterpolate(
   B=@n22
   Alpha=DynamicParameter() {defaultValue=<rgba 0 1 1 1>, paramNames [FX_Progress, FX_Temperature, FX_Distort, FX_Grid_Strength]}.FX_Distort
  )
  n29 = ParticleColor()
  n30 = SphereMask(
   A=TextureCoordinate()
   B=Constant() {r=0.5}
   Radius=ScalarParameter() {defaultValue=0.5, parameterName=Radius}
   Hardness=ScalarParameter() {parameterName=Hardness}
  )
  Multiply(
   A=Multiply(A=Saturate(None=@n24.RGB), B=Multiply(A=@n29.A, B=@n30))
   B=DepthFade(FadeDistance=ScalarParameter() {defaultValue=100, parameterName=DepthFade_Distance})
  )
 ]
}
